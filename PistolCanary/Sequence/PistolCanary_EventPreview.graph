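# PistolCanary_EventPreview is a Level Sequence that lays out three sections back to back over about one second at a 60 fps display rate.
# It has no bindings of its own, only three sub tracks named PistolCanary_Section_01 to 03, each holding one sub section.
# The sections run contiguously in numeric order, though the tracks are stored out of order, and the middle section is the longest.
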LevelSequence /PistolCanary/Sequence/PistolCanary_EventPreview
bindings []
display_rate {denominator=1, numerator=60}
tick_resolution {denominator=1, numerator=24000}
tracks [
 {type=MovieSceneSubTrack, display_name=PistolCanary_Section_03, sections [{type=MovieSceneSubSection, end_frame=24209, start_frame=17661}]}
 {type=MovieSceneSubTrack, display_name=PistolCanary_Section_01, sections [{type=MovieSceneSubSection, end_frame=5813, start_frame=0}]}
 {type=MovieSceneSubTrack, display_name=PistolCanary_Section_02, sections [{type=MovieSceneSubSection, end_frame=17661, start_frame=5813}]}
]
