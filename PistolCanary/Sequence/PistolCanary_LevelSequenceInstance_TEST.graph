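# Test level sequence for PistolCanary that raises four rock spire monolith actors and lowers them again over 150 frames at 30 fps.
# Each monolith has its own transform track with identical Location Z keys, so all four peak together at the midpoint and end slightly below their start.
# Two cube mesh actors have spawn, scale-up transform and attach tracks, the first also holding a fixed location.
# A SceneRoot actor is bound without tracks.
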LevelSequence /PistolCanary/Sequence/PistolCanary_LevelSequenceInstance_TEST
bindings [
 Cube {
  id=#1
  possessed_class=FortStaticMeshActor
  tracks [
   {type=MovieSceneSpawnTrack, sections [{type=MovieSceneSpawnSection, channels [None {type=MovieSceneScriptingBoolChannel, default=true}]}]}
   {
    type=MovieScene3DTransformTrack
    sections [
     {
      type=MovieScene3DTransformSection
      channels [
       Location.X {type=MovieSceneScriptingDoubleChannel, default=1714, const [1714,1]}
       Location.Y {type=MovieSceneScriptingDoubleChannel, default=5227, const [5227,1]}
       Location.Z {type=MovieSceneScriptingDoubleChannel, default=3431, const [3431,1]}
       Rotation.X {type=MovieSceneScriptingDoubleChannel, default=0}
       Rotation.Y {type=MovieSceneScriptingDoubleChannel, default=0}
       Rotation.Z {type=MovieSceneScriptingDoubleChannel, default=0}
       Scale.X {type=MovieSceneScriptingDoubleChannel, default=0}
       Scale.Y {type=MovieSceneScriptingDoubleChannel, default=0}
       Scale.Z {type=MovieSceneScriptingDoubleChannel, default=0}
      ]
     }
     {
      type=MovieScene3DTransformSection
      channels [
       Location.X {type=MovieSceneScriptingDoubleChannel, default=0}
       Location.Y {type=MovieSceneScriptingDoubleChannel, default=0}
       Location.Z {type=MovieSceneScriptingDoubleChannel, default=0}
       Rotation.X {type=MovieSceneScriptingDoubleChannel, default=0}
       Rotation.Y {type=MovieSceneScriptingDoubleChannel, default=0}
       Rotation.Z {type=MovieSceneScriptingDoubleChannel, default=0}
       Scale.X {type=MovieSceneScriptingDoubleChannel, default=10}
       Scale.Y {type=MovieSceneScriptingDoubleChannel, default=10}
       Scale.Z {type=MovieSceneScriptingDoubleChannel, default=10}
      ]
     }
    ]
   }
   {type=MovieScene3DAttachTrack, sections [{type=MovieScene3DAttachSection, end_frame=150, start_frame=0}]}
  ]
 }
 PistolCanary_RockSpireMonolith3 {
  id=#2
  possessed_class=PistolCanary_Monolith_Spire_C
  tracks [{
    type=MovieScene3DTransformTrack
    sections [{
      type=MovieScene3DTransformSection
      channels [
       Location.X {type=MovieSceneScriptingDoubleChannel, default=0}
       Location.Y {type=MovieSceneScriptingDoubleChannel, default=0}
       Location.Z {type=MovieSceneScriptingDoubleChannel, default=5885, keys [[0,-109.3164],[75,5885],[150,-218.6328]]}
       Rotation.X {type=MovieSceneScriptingDoubleChannel, default=0}
       Rotation.Y {type=MovieSceneScriptingDoubleChannel, default=0}
       Rotation.Z {type=MovieSceneScriptingDoubleChannel, default=0}
       Scale.X {type=MovieSceneScriptingDoubleChannel, default=0}
       Scale.Y {type=MovieSceneScriptingDoubleChannel, default=0}
       Scale.Z {type=MovieSceneScriptingDoubleChannel, default=0}
      ]
     }]
   }]
 }
 PistolCanary_SceneRoot {id=#3, possessed_class=FortStaticMeshActor, tracks []}
 PistolCanary_RockSpireMonolith2 {
  id=#4
  possessed_class=PistolCanary_Monolith_Spire_C
  tracks [{
    type=MovieScene3DTransformTrack
    sections [{
      type=MovieScene3DTransformSection
      channels [
       Location.X {type=MovieSceneScriptingDoubleChannel, default=0}
       Location.Y {type=MovieSceneScriptingDoubleChannel, default=0}
       Location.Z {type=MovieSceneScriptingDoubleChannel, default=5885, keys [[0,-109.3164],[75,5885],[150,-218.6328]]}
       Rotation.X {type=MovieSceneScriptingDoubleChannel, default=0}
       Rotation.Y {type=MovieSceneScriptingDoubleChannel, default=0}
       Rotation.Z {type=MovieSceneScriptingDoubleChannel, default=0}
       Scale.X {type=MovieSceneScriptingDoubleChannel, default=0}
       Scale.Y {type=MovieSceneScriptingDoubleChannel, default=0}
       Scale.Z {type=MovieSceneScriptingDoubleChannel, default=0}
      ]
     }]
   }]
 }
 PistolCanary_RockSpireMonolith1 {
  id=#5
  possessed_class=PistolCanary_Monolith_Spire_C
  tracks [{
    type=MovieScene3DTransformTrack
    sections [{
      type=MovieScene3DTransformSection
      channels [
       Location.X {type=MovieSceneScriptingDoubleChannel, default=0}
       Location.Y {type=MovieSceneScriptingDoubleChannel, default=0}
       Location.Z {type=MovieSceneScriptingDoubleChannel, default=5885, keys [[0,-109.3164],[75,5885],[150,-218.6328]]}
       Rotation.X {type=MovieSceneScriptingDoubleChannel, default=0}
       Rotation.Y {type=MovieSceneScriptingDoubleChannel, default=0}
       Rotation.Z {type=MovieSceneScriptingDoubleChannel, default=0}
       Scale.X {type=MovieSceneScriptingDoubleChannel, default=0}
       Scale.Y {type=MovieSceneScriptingDoubleChannel, default=0}
       Scale.Z {type=MovieSceneScriptingDoubleChannel, default=0}
      ]
     }]
   }]
 }
 PistolCanary_RockSpireMonolith4 {
  id=#6
  possessed_class=PistolCanary_Monolith_Spire_C
  tracks [{
    type=MovieScene3DTransformTrack
    sections [{
      type=MovieScene3DTransformSection
      channels [
       Location.X {type=MovieSceneScriptingDoubleChannel, default=0}
       Location.Y {type=MovieSceneScriptingDoubleChannel, default=0}
       Location.Z {type=MovieSceneScriptingDoubleChannel, default=5885, keys [[0,-109.3164],[75,5885],[150,-218.6328]]}
       Rotation.X {type=MovieSceneScriptingDoubleChannel, default=0}
       Rotation.Y {type=MovieSceneScriptingDoubleChannel, default=0}
       Rotation.Z {type=MovieSceneScriptingDoubleChannel, default=0}
       Scale.X {type=MovieSceneScriptingDoubleChannel, default=0}
       Scale.Y {type=MovieSceneScriptingDoubleChannel, default=0}
       Scale.Z {type=MovieSceneScriptingDoubleChannel, default=0}
      ]
     }]
   }]
 }
 Cube2 {
  id=#7
  possessed_class=FortStaticMeshActor
  tracks [
   {type=MovieSceneSpawnTrack, sections [{type=MovieSceneSpawnSection, channels [None {type=MovieSceneScriptingBoolChannel, default=true}]}]}
   {
    type=MovieScene3DTransformTrack
    sections [{
      type=MovieScene3DTransformSection
      channels [
       Location.X {type=MovieSceneScriptingDoubleChannel, default=0}
       Location.Y {type=MovieSceneScriptingDoubleChannel, default=0}
       Location.Z {type=MovieSceneScriptingDoubleChannel, default=0}
       Rotation.X {type=MovieSceneScriptingDoubleChannel, default=0}
       Rotation.Y {type=MovieSceneScriptingDoubleChannel, default=0}
       Rotation.Z {type=MovieSceneScriptingDoubleChannel, default=0}
       Scale.X {type=MovieSceneScriptingDoubleChannel, default=10}
       Scale.Y {type=MovieSceneScriptingDoubleChannel, default=10}
       Scale.Z {type=MovieSceneScriptingDoubleChannel, default=10}
      ]
     }]
   }
   {type=MovieScene3DAttachTrack, sections [{type=MovieScene3DAttachSection, end_frame=150, start_frame=0}]}
  ]
 }
]
display_rate {denominator=1, numerator=30}
tick_resolution {denominator=1, numerator=24000}
tracks []
